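# Looping music Sound Cue for Cosmos gameplay that layers low and high intensity tracks, routed through the Fort_Music_NPC sound class.
# A Mixer plays both looping waves at once, each behind a Continuous Modulator on volume.
# The "Unaware" parameter drives the low loop and "Threatened" drives the high loop, both defaulting to zero, so runtime parameter values set the blend between states.
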SoundCue /CosmosGameplay/Audio/Music/Cosmos_Music_Loop_Cue
root=Mixer_1
nodes [
 {id=Mixer_1, children [ModulatorContinuous_0, ModulatorContinuous_1], properties {inputVolume [1,1]}}
 {id=ModulatorContinuous_0, children [WavePlayer_0], properties {volumeModulationParams {default=0, parameterName=Unaware}}}
 {id=ModulatorContinuous_1, children [WavePlayer_1], properties {volumeModulationParams {default=0, parameterName=Threatened}}}
 {id=WavePlayer_0, properties {bLooping=true, soundWaveAssetPtr=&/Cosmos_Loop_Low}}
 {id=WavePlayer_1, properties {bLooping=true, soundWaveAssetPtr=&/Cosmos_HighLoop_01}}
]
properties {maxDistance=2097152, soundClassObject=&Fort_Music_NPC, volumeMultiplier=1.2}
