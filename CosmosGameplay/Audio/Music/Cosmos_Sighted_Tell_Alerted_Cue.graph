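# Sound cue for the Cosmos "sighted tell alerted" music stinger, routed to the NPC music sound class.
# A switch on the integer parameter Range sends all three inputs to one random node, which picks evenly between two stinger variations.
# Overridden attenuation disables volume falloff but keeps distance low-pass filtering and occlusion.
# A dedicated concurrency asset governs simultaneous playback.
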SoundCue /CosmosGameplay/Audio/Music/Cosmos_Sighted_Tell_Alerted_Cue
root=Switch_1
nodes [
 {id=Random_4, children [WavePlayer_3,WavePlayer_4], properties {weights [1,1]}}
 {id=Switch_1, children [Random_4,Random_4,Random_4], properties {intParameterName=Range}}
 {id=WavePlayer_3, properties {soundWaveAssetPtr=&/Cosmos_Stinger_02a}}
 {id=WavePlayer_4, properties {soundWaveAssetPtr=&/Cosmos_Stinger_02b}}
]
properties {
 bOverrideAttenuation=true
 maxDistance=2097152
 soundClassObject=&Fort_Music_NPC
 volumeMultiplier=1
 attenuationOverrides {
  bAttenuate=false
  bAttenuateWithLPF=true
  bEnableOcclusion=true
  lPFFrequencyAtMax=6000
  lPFRadiusMax=80000
  lPFRadiusMin=40000
  occlusionLowPassFilterFrequency=6000
  occlusionTraceChannel=ECC_GameTraceChannel11
  stereoSpread=5000
 }
 concurrencyOverrides {maxCount=1, resolutionRule=PreventNew}
 concurrencySet [&/Cosmos_Sighted_Tell_Conc]
}
