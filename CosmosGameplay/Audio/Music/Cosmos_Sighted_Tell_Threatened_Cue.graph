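# Music stinger cue for the Cosmos "sighted tell" in its threatened variant, played as non-attenuated audio.
# An integer Range switch picks between two random pairs: stinger 01 a/b for the first two inputs, including the default, and stinger 03 a/b for the last.
# Attenuation overrides keep volume constant over distance while adding distance low-pass filtering and occlusion.
# A shared Sighted Tell concurrency set governs simultaneous playback.
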SoundCue /CosmosGameplay/Audio/Music/Cosmos_Sighted_Tell_Threatened_Cue
root=Switch_1
nodes [
 {id=Random_2, children [WavePlayer_5,WavePlayer_6], properties {weights [1,1]}}
 {id=Random_3, children [WavePlayer_1,WavePlayer_2], properties {weights [1,1]}}
 {id=Switch_1, children [Random_3,Random_3,Random_2], properties {intParameterName=Range}}
 {id=WavePlayer_1, properties {soundWaveAssetPtr=&/Cosmos_Stinger_01a}}
 {id=WavePlayer_2, properties {soundWaveAssetPtr=&/Cosmos_Stinger_01b}}
 {id=WavePlayer_5, properties {soundWaveAssetPtr=&/Cosmos_Stinger_03a}}
 {id=WavePlayer_6, properties {soundWaveAssetPtr=&/Cosmos_Stinger_03b}}
]
properties {
 bOverrideAttenuation=true
 maxDistance=2097152
 priority=2
 soundClassObject=&Fort_WeaponRanged_Fire_3P_PSM
 volumeMultiplier=0.6
 attenuationOverrides {
  bAttenuate=false
  bAttenuateWithLPF=true
  bEnableOcclusion=true
  lPFFrequencyAtMax=6000
  lPFRadiusMax=80000
  lPFRadiusMin=40000
  occlusionLowPassFilterFrequency=6000
  occlusionTraceChannel=ECC_GameTraceChannel11
  stereoSpread=5000
 }
 concurrencyOverrides {maxCount=1, resolutionRule=PreventNew}
 concurrencySet [&/Cosmos_Sighted_Tell_Conc]
}
