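; A Quartz music track component for Cosmos that drives track parameters from gameplay awareness states.
; On each data tracker update it reads event values for the Unaware, Alerted and Threatened tags, summing Unaware with Alerted and capping both results at one.
; It eases two stored values toward these targets with FInterpTo, then sends Unaware to track parameter 0 and Threatened to parameter 1.
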(event EventBeginPlay
  (|CallParentFunction) ; This node is disabled and will not be called.\nDrag off pins to build functionality.
)

(event EventTick
  (params (DeltaSeconds :type Float%20%28single-precision%29)))

(fn OnDataTrackerUpdate
  (Utilities|FlowControl|Sequence
    (:then_0
      (Variables|Default|SetInterp_Threatened (Math|Interpolation|FInterpTo (Variables|Default|GetInterp_Threatened) (Math|Float|Min(Float) (Class|FortGameplayDataTrackerComponent|GetEventValue (CallFunction|GetDataTrackerComponent) (Variables|Default|GetTag_Threatened)) 1.0) (Utilities|Time|GetWorldDeltaSeconds) 2.0))
      (Variables|Default|SetInterp_Unaware (Math|Interpolation|FInterpTo (Variables|Default|GetInterp_Unaware) (Math|Float|Min(Float) (+ (Class|FortGameplayDataTrackerComponent|GetEventValue (CallFunction|GetDataTrackerComponent) (Variables|Default|GetTag_Unaware)) (Class|FortGameplayDataTrackerComponent|GetEventValue (CallFunction|GetDataTrackerComponent) (Variables|Default|GetTag_Alerted))) 1.0) (Utilities|Time|GetWorldDeltaSeconds) 2.0)))
    (:then_1
      (Control|SetTrackParameter 0 (Variables|Default|GetInterp_Unaware))
      (Control|SetTrackParameter 1 (Variables|Default|GetInterp_Threatened)))))
